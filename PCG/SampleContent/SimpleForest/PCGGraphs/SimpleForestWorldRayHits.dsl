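; Sample PCG graph that scatters a simple forest of trees, boulders and seedling grass on surfaces found by a World Ray Hit Query.
; Three Surface Samplers share the query, each feeding random transforms and a weighted spawner of hierarchical instanced meshes.
; Trees get fixed bounds, rocks are self-pruned and subtract trees, and grass subtracts both through binary Difference nodes, layering placement by priority.
; Grass spawns without collision, while trees and rocks use query-only collision.
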(graph SimpleForestWorldRayHits
  (bind WorldRayHitQuery_0 (World%20Ray%20Hit%20Query))
  ;[0 0 0] Trees
  (bind SurfaceSampler_3 (Surface%20Sampler :PointExtents (2e+02 2e+02 5e+02) :PointsPerSquaredMeter 0.2 :PointSteepness 0.5)) ; Distribute points over surface
  (bind TransformPoints_27 (Transform%20Points :bAbsoluteRotation true :RotationMax (struct :Pitch 5.0 :Yaw 1.8e+02 :Roll 5.0) :RotationMin (struct :Pitch -5.0 :Yaw -1.8e+02 :Roll -5.0) :ScaleMax (1.2 1.2 1.2) :ScaleMin (0.7 0.7 0.7))) ; Apply random scale and rotation
  (bind BoundsModifier_28 (Bounds%20Modifier :bAffectSteepness true :BoundsMax (32.0 32.0 1e+02) :BoundsMin (-32.0 -32.0 -1e+02) :Mode Set)) ; Set points boundings
  (bind StaticMeshSpawner_5 (Static%20Mesh%20Spawner :bApplyMeshBoundsToPoints false :MeshSelectorParameters (class:/Script/PCG.PCGMeshSelectorWeighted :MeshEntries [(struct :Descriptor (struct :StaticMesh /PCG/SampleContent/SimpleForest/Meshes/PCG_Tree_01.PCG_Tree_01 :ComponentClass /Script/Engine.HierarchicalInstancedStaticMeshComponent :Mobility Movable :BodyInstance (struct :VelocitySolverIterationCount 1 :ObjectType ECC_WorldDynamic :CollisionEnabled QueryOnly :CollisionProfileName BlockAllDynamic :CollisionResponses (struct :ResponseArray [(struct :Channel FCC_Trace_WeaponCapsule) (struct :Channel FCC_Object_PawnAI) (struct :Channel FCC_Trace_WeaponMulti) (struct :Channel FCC_Trace_Weapon) (struct :Channel FCC_Object_Projectile) (struct :Channel FCC_Object_ProjectileIgnoreAI) (struct :Channel FCC_Object_ProjectileMesh) (struct :Channel FCC_Trace_WeaponPawnAndWorldMulti)]) :bAutoWeld true) :bUseDefaultCollision true :bGenerateOverlapEvents true :bCanEverAffectNavigation true)) (struct :Descriptor (struct :StaticMesh /PCG/SampleContent/SimpleForest/Meshes/PCG_Tree_02.PCG_Tree_02 :ComponentClass /Script/Engine.HierarchicalInstancedStaticMeshComponent :Mobility Movable :BodyInstance (struct :VelocitySolverIterationCount 1 :ObjectType ECC_WorldDynamic :CollisionEnabled QueryOnly :CollisionProfileName BlockAllDynamic :CollisionResponses (struct :ResponseArray [(struct :Channel FCC_Trace_WeaponCapsule) (struct :Channel FCC_Object_PawnAI) (struct :Channel FCC_Trace_WeaponMulti) (struct :Channel FCC_Trace_Weapon) (struct :Channel FCC_Object_Projectile) (struct :Channel FCC_Object_ProjectileIgnoreAI) (struct :Channel FCC_Object_ProjectileMesh) (struct :Channel FCC_Trace_WeaponPawnAndWorldMulti)]) :bAutoWeld true) :bUseDefaultCollision true :bGenerateOverlapEvents true :bCanEverAffectNavigation true)) (struct :Descriptor (struct :StaticMesh /PCG/SampleContent/SimpleForest/Meshes/PCG_Tree_03.PCG_Tree_03 :ComponentClass /Script/Engine.HierarchicalInstancedStaticMeshComponent :Mobility Movable :BodyInstance (struct :VelocitySolverIterationCount 1 :ObjectType ECC_WorldDynamic :CollisionEnabled QueryOnly :CollisionProfileName BlockAllDynamic :CollisionResponses (struct :ResponseArray [(struct :Channel FCC_Trace_WeaponCapsule) (struct :Channel FCC_Object_PawnAI) (struct :Channel FCC_Trace_WeaponMulti) (struct :Channel FCC_Trace_Weapon) (struct :Channel FCC_Object_Projectile) (struct :Channel FCC_Object_ProjectileIgnoreAI) (struct :Channel FCC_Object_ProjectileMesh) (struct :Channel FCC_Trace_WeaponPawnAndWorldMulti)]) :bAutoWeld true) :bUseDefaultCollision true :bGenerateOverlapEvents true :bCanEverAffectNavigation true))]))) ; Spawn trees meshes
  ;[0 0 0] Grass
  (bind SurfaceSampler_7 (Surface%20Sampler :PointExtents (1e+01 1e+01 1e+01) :PointsPerSquaredMeter 1.0 :PointSteepness 0.5)) ; Distribute points over surface
  (bind Difference_11 (Difference :DensityFunction Binary)) ; Remove points from another set (trees and rocks)
  (bind ToPoint_30 (To%20Point))
  (bind TransformPoints_29 (Transform%20Points :RotationMax (struct :Yaw 1.8e+02) :RotationMin (struct :Yaw -1.8e+02) :ScaleMax (0.75 0.75 0.75) :ScaleMin (0.5 0.5 0.5))) ; Apply random scale and rotation
  (bind StaticMeshSpawner_9 (Static%20Mesh%20Spawner :bApplyMeshBoundsToPoints false :MeshSelectorParameters (class:/Script/PCG.PCGMeshSelectorWeighted :MeshEntries [(struct :Descriptor (struct :StaticMesh /PCG/SampleContent/SimpleForest/Meshes/PCG_Seedling_01.PCG_Seedling_01 :ComponentClass /Script/Engine.HierarchicalInstancedStaticMeshComponent :Mobility Movable :BodyInstance (struct :VelocitySolverIterationCount 1 :CollisionEnabled NoCollision :CollisionProfileName NoCollision :CollisionResponses (struct :ResponseArray [(struct :Channel Visibility :Response ECR_Ignore) (struct :Channel Camera :Response ECR_Ignore)]) :bAutoWeld true) :bGenerateOverlapEvents true :bCanEverAffectNavigation true) :Weight 400)]))) ; Spawn grass mesh
  ;[0 0 0] Rocks
  (bind SurfaceSampler_13 (Surface%20Sampler :PointExtents (1e+02 75.0 6e+01) :PointsPerSquaredMeter 0.02 :PointSteepness 0.5)) ; Distribute points over surface
  (bind TransformPoints_26 (Transform%20Points :RotationMax (struct :Yaw 1.8e+02) :RotationMin (struct :Yaw -1.8e+02) :ScaleMax (3.0 3.0 3.0) :ScaleMin (2.0 2.0 2.0))) ; Apply random scale and rotation
  (bind SelfPruning_19 (Self%20Pruning)) ; Remove overlapping points
  (bind Difference_17 (Difference :DensityFunction Binary)) ; Remove points from another set (trees)
  (bind ToPoint_0 (To%20Point))
  (bind StaticMeshSpawner_15 (Static%20Mesh%20Spawner :bApplyMeshBoundsToPoints false :MeshSelectorParameters (class:/Script/PCG.PCGMeshSelectorWeighted :MeshEntries [(struct :Descriptor (struct :StaticMesh /PCG/SampleContent/SimpleForest/Meshes/PCG_Boulder_02.PCG_Boulder_02 :ComponentClass /Script/Engine.HierarchicalInstancedStaticMeshComponent :Mobility Movable :BodyInstance (struct :VelocitySolverIterationCount 1 :ObjectType ECC_WorldDynamic :CollisionEnabled QueryOnly :CollisionProfileName BlockAllDynamic :CollisionResponses (struct :ResponseArray [(struct :Channel FCC_Trace_WeaponCapsule) (struct :Channel FCC_Object_PawnAI) (struct :Channel FCC_Trace_WeaponMulti) (struct :Channel FCC_Trace_Weapon) (struct :Channel FCC_Object_Projectile) (struct :Channel FCC_Object_ProjectileIgnoreAI) (struct :Channel FCC_Object_ProjectileMesh) (struct :Channel FCC_Trace_WeaponPawnAndWorldMulti)]) :bAutoWeld true) :bUseDefaultCollision true :bGenerateOverlapEvents true :bCanEverAffectNavigation true))]))) ; Spawn rock mesh
  (connect TransformPoints_27 Out BoundsModifier_28 In)
  (connect BoundsModifier_28 Out Difference_11 Differences)
  (connect Difference_17 Out Difference_11 Differences)
  (connect SurfaceSampler_7 Out Difference_11 Source)
  (connect BoundsModifier_28 Out Difference_17 Differences)
  (connect SelfPruning_19 Out Difference_17 Source)
  (connect TransformPoints_26 Out SelfPruning_19 In)
  (connect BoundsModifier_28 Out StaticMeshSpawner_5 In)
  (connect TransformPoints_29 Out StaticMeshSpawner_9 In)
  (connect ToPoint_0 Out StaticMeshSpawner_15 In)
  (connect WorldRayHitQuery_0 Out SurfaceSampler_3 Surface)
  (connect WorldRayHitQuery_0 Out SurfaceSampler_7 Surface)
  (connect WorldRayHitQuery_0 Out SurfaceSampler_13 Surface)
  (connect Difference_17 Out ToPoint_0 In)
  (connect Difference_11 Out ToPoint_30 In)
  (connect SurfaceSampler_13 Out TransformPoints_26 In)
  (connect SurfaceSampler_3 Out TransformPoints_27 In)
  (connect ToPoint_30 Out TransformPoints_29 In))
